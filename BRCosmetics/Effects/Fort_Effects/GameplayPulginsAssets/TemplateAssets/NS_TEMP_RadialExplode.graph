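# Template Niagara radial explosion burst, parameterized by user color, scale, time, intensity and an element index.
# Expanding sphere meshes, a glow sprite with light and bright streaks form the blast, with optional curl-noise sparks that fall and distance-cull.
# The element index gates an extra elemental sphere and selects material and timing values from arrays.
# An optional debug ring shows the scale radius, and the PostMat mesh emitter is disabled.
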NiagaraSystem /BRCosmetics/Effects/Fort_Effects/GameplayPulginsAssets/TemplateAssets/NS_TEMP_RadialExplode
dependencies {
 usedDataInterfaces [
  &NiagaraDataInterfaceCurve
  &NiagaraDataInterfaceVector2DCurve
  &NiagaraDataInterfaceArrayFloat
  /Script/Niagara.NiagaraDataInterfaceArrayFloat4
  &NiagaraDataInterfaceVectorCurve
 ]
 usedDynamicInputs [&Multiply_VectorByFloat, /Niagara/DynamicInputs/Scalability/ScalabilityDistanceBasedFloat]
 usedModules [&SystemState]
}
emitters {
 BrightStreaks {
  emitterUpdateScript [
   &EmitterState()
   &SpawnBurst_Instantaneous(Spawn Count=20, Use Spawn Probability=false, UseLoopCountLimit=false)
  ]
  particleSpawnScript [
   /SetVariables(Particles.SizeMax=450, Particles.SizeMin=150) {isSetParameters=true}
   /SetVariables(Particles.SizeCurve:NiagaraFloat=&Lerp_Float(
     A=$Particles.SizeMin
     B=$Particles.SizeMax
     Alpha=&FloatFromCurve(
      FloatCurve=DataInterface() {propertyValues {Curve {keys [
          {interpMode=RCIM_Cubic, tangentMode=RCTM_Break, time=0, value=0}
          {arriveTangent=2.305, interpMode=RCIM_Cubic, tangentMode=RCTM_Break, time=1, value=1}
         ]}}}
      CurveIndex=&RandomRangeFloat(Minimum=0, Maximum=1)
     )
    )) {isSetParameters=true}
   &InitializeParticle(
    Color=$User.FXColor
    Lifetime=&Multiply_Float(A=&RandomRangeFloat(Minimum=0.3, Maximum=0.6), B=$User.FXTime)
    Sprite Size=&Multiply_Vector2D_ByFloat(
     Float=&Multiply_Float(A=0.003, B=$User.FXScale)
     Vector2D=&RandomRangeVector2D(Minimum=<75 100>, Maximum=<100 210>)
    )
    Sprite Size Mode=::NewEnumerator4
    UsePositionOffset=false
    Write Material Random=true
   )
   &ShapeLocation(
    MaskSpawnGroup=false
    Override Seed=false
    Sphere Radius=&Multiply_Float(A=0.5, B=$User.FXScale)
    UseOwnerScale=true
   )
   &AddVelocity(
    Module.Use Falloff Distance=false
    Override Seed=false
    Velocity Mode=::NewEnumerator1
    Velocity Speed=&Multiply_Float(
     A=&Lerp_Float(
      A=300
      B=400
      Alpha=&FloatFromCurve(
       FloatCurve=DataInterface() {propertyValues {Curve {keys [{interpMode=RCIM_Cubic, tangentMode=RCTM_Break, time=0, value=0}, {interpMode=RCIM_Cubic, tangentMode=RCTM_Break, time=1, value=1}]}}}
       CurveIndex=&RandomRangeFloat(Minimum=0, Maximum=1)
      )
     )
     B=1
    )
   )
   &SubUVAnimation(End Frame=3, SubUV Animation Mode:ENiagara_SubUVLookupMode=::NewEnumerator2)
   &CameraOffset(Camera Offset Amount=50)
  ]
  particleUpdateScript [
   &ParticleState()
   &Drag(Drag=4, Use Linear Drag=true, Use Rotational Drag=false)
   &SolveForcesAndVelocity(Clamp Velocity=false, Limit Acceleration=false)
   &ScaleColor(
    Scale Alpha=&FloatFromCurve(
     FloatCurve=DataInterface() {propertyValues {Curve {keys [
         {interpMode=RCIM_Cubic, leaveTangent=9.1535, leaveTangentWeight=0.0962, tangentMode=RCTM_Break, time=0, value=0}
         {interpMode=RCIM_Cubic, tangentMode=RCTM_Break, time=0.1, value=1}
         {interpMode=RCIM_Cubic, tangentMode=RCTM_Break, time=1, value=0}
        ]}}}
     Scale Curve=&RandomRangeFloat(Minimum=0.4, Maximum=0.75)
    )
    Scale RGB=&VectorFromFloat(Value=&Lerp_Float(
      A=1
      B=20
      Alpha=&FloatFromCurve(FloatCurve=DataInterface() {propertyValues {Curve {keys [
           {interpMode=RCIM_Cubic, tangentMode=RCTM_User, time=0, value=0}
           {arriveTangent=2.2791, interpMode=RCIM_Cubic, leaveTangent=2.2791, tangentMode=RCTM_User, time=1, value=1}
          ]}}})
     ))
    ScaleA=true
    ScaleRGB=true
   )
   &ScaleSpriteSize(Scale Factor=&Vector2DFromCurve(
     Vector2Curve=DataInterface() {propertyValues {XCurve {keys [<at 0.0036 1.6897>,<at 0.9967 0.0582>]}, YCurve {keys [<at 0 1>,<at 1.002 3.7227>]}}}
     ScaleCurve=<1 1>
    ))
   &DynamicMaterialParameters(
    Index 0 Param 2=1
    Index 0 Param 3=1
    Index 0 Param 4=&FloatFromCurve(
     FloatCurve=DataInterface() {propertyValues {Curve {keys [{interpMode=RCIM_Cubic, tangentMode=RCTM_Break, time=0.3, value=1}, {interpMode=RCIM_Cubic, tangentMode=RCTM_Break, time=1, value=0.25}]}}}
     Scale Curve=&RandomRangeFloat(Minimum=0.75, Maximum=1)
    )
    Module.Param0WriteEnabled=true
    Module.Param1WriteEnabled=true
    Module.Param2WriteEnabled=true
    Module.Param3WriteEnabled=true
   )
  ]
  rendererClasses [&NiagaraSpriteRendererProperties]
  renderers [{rendererClass=&NiagaraSpriteRendererProperties}]
 }
 DEBUG {
  emitterUpdateScript [
   &EmitterState()
   &SpawnBurst_Instantaneous(
    Spawn Count=1000
    Spawn Probability=&MakeCustomFloatFromBool(Bool=$User.ShowDebugRing, True Float=1, False Float=0)
    Use Spawn Probability=true
    UseLoopCountLimit=false
   )
  ]
  particleSpawnScript [
   &InitializeParticle(
    Color=<rgba 1 0 0.999 1>
    Lifetime=$User.FXTime
    Sprite Size Mode=::NewEnumerator3
    Uniform Sprite Size=15
    UsePositionOffset=false
    Write Material Random=true
   )
   &ShapeLocation(MaskSpawnGroup=false, Override Seed=false, Ring Radius=$User.FXScale, Shape Primitive=::NewEnumerator4, UseOwnerScale=true)
  ]
  particleUpdateScript [&ParticleState()]
  rendererClasses [&NiagaraSpriteRendererProperties]
  renderers [{rendererClass=&NiagaraSpriteRendererProperties}]
 }
 LightAndGlow {
  emitterUpdateScript [
   &EmitterState()
   &SpawnBurst_Instantaneous(Use Spawn Probability=false, UseLoopCountLimit=false)
  ]
  particleSpawnScript [
   &InitializeParticle(
    Color=$User.FXColor
    Lifetime=&Multiply_Float(A=0.5, B=$User.FXTime)
    Sprite Rotation Mode=::NewEnumerator2
    Sprite Size Mode=::NewEnumerator3
    Uniform Sprite Size=&Multiply_Float(A=4, B=$User.FXScale)
    UsePositionOffset=false
    Write Material Random=true
   )
   &DynamicMaterialParameters(Module.Index0Param0Write=true, Module.Index0Param1Write=true, Module.Index0Param2Write=true, Module.Index0Param3Write=true)
   &CameraOffset(Camera Offset Amount=&Multiply_Float(A=0.3, B=$User.FXScale))
  ]
  particleUpdateScript [
   &ParticleState()
   &ScaleColor(
    Scale Alpha=&FloatFromCurve(FloatCurve=DataInterface() {propertyValues {Curve {keys [{interpMode=RCIM_Cubic, tangentMode=RCTM_User, time=0, value=1}, {interpMode=RCIM_Cubic, tangentMode=RCTM_User, time=1, value=0}]}}})
    Scale RGB=&VectorFromFloat(Value=6)
    ScaleA=true
    ScaleRGB=true
   )
   &FN_Light_Attributes(
    Color=$Particles.Initial.Color
    Create Color Attribute=true
    Create Exponent Attribute=true
    Create Position Attribute=true
    Create Radius Attribute=true
    Enable Particle Light Enable=true
    Light Enabled=true
    Position=&AddVectorToPosition(Position=$Particles.Position, Vector=<0 0 100>)
    Radius=&Multiply_Float(A=2, B=$User.FXScale)
    Scale Color=&FloatFromCurve(
     FloatCurve=DataInterface() {propertyValues {Curve {keys [{interpMode=RCIM_Cubic, tangentMode=RCTM_User, time=0, value=1}, {interpMode=RCIM_Cubic, tangentMode=RCTM_User, time=1, value=0}]}}}
     Scale Curve=&Lerp_Float(
      A=1
      B=20
      Alpha=&FloatFromCurve(FloatCurve=DataInterface() {propertyValues {Curve {keys [
           {arriveTangent=9.0725, interpMode=RCIM_Cubic, leaveTangent=9.0725, tangentMode=RCTM_User, time=0, value=0}
           {arriveTangentWeight=0.0453, interpMode=RCIM_Cubic, leaveTangentWeight=0.3168, tangentMode=RCTM_User, time=0.05, value=0.9847}
           {arriveTangent=-1.146, leaveTangent=-1.146, time=0.5, value=0.15}
           {arriveTangent=-0.0405, interpMode=RCIM_Cubic, leaveTangent=-0.0405, tangentMode=RCTM_User, time=1, value=0}
          ]}}})
     )
    )
   )
   &SolveForcesAndVelocity(Clamp Velocity=false, Limit Acceleration=false)
  ]
  rendererClasses [&NiagaraSpriteRendererProperties, &NiagaraLightRendererProperties]
  renderers [
   {rendererClass=&NiagaraSpriteRendererProperties}
   {rendererClass=&NiagaraLightRendererProperties, rendererIndex=1}
  ]
 }
 PostMat {
  bEnabled=false
  emitterUpdateScript [
   &EmitterState()
   &SpawnBurst_Instantaneous(Use Spawn Probability=false, UseLoopCountLimit=false)
  ]
  particleSpawnScript [&InitializeParticle(
    Color Mode=::NewEnumerator0
    Mesh Scale Mode=::NewEnumerator3
    Mesh Uniform Scale=70
    Sprite Size Mode=::NewEnumerator3
    UsePositionOffset=false
    Write Material Random=true
   )]
  particleUpdateScript [
   &ParticleState()
   &DynamicMaterialParameters(
    Index 0 Param 1=&MakeFloatFromVector(Vector=&Multiply_Vector(A=$Particles.Scale, B=<1 1 1>), Channel=::NewEnumerator0)
    Index 0 Param 2=&MakeFloatFromVector(
     Vector=&TransformPosition(Position=$Particles.Position, Source Space=::Simulation, Destination Space=::World)
     Channel=::NewEnumerator0
    )
    Index 0 Param 3=&MakeFloatFromVector(
     Vector=&TransformPosition(Position=$Particles.Position, Source Space=::Simulation, Destination Space=::World)
     Channel=::NewEnumerator1
    )
    Index 0 Param 4=&MakeFloatFromVector(
     Vector=&TransformPosition(Position=$Particles.Position, Source Space=::Simulation, Destination Space=::World)
     Channel=::NewEnumerator2
    )
    Index 1 Param 1=&FloatFromCurve(FloatCurve=DataInterface() {propertyValues {Curve {keys [
         {interpMode=RCIM_Cubic, leaveTangent=-8.4786, tangentMode=RCTM_Break, time=0, value=2.6}
         {arriveTangent=-2.4925, interpMode=RCIM_Cubic, tangentMode=RCTM_Break, time=1, value=-1.7}
        ]}}})
    Index 1 Param 2=&FloatFromCurve(FloatCurve=DataInterface() {propertyValues {Curve {keys [
         {arriveTangent=4.7488, interpMode=RCIM_Cubic, leaveTangent=4.7488, tangentMode=RCTM_User, time=0.2, value=0.1}
         {arriveTangent=5.8572, interpMode=RCIM_Cubic, leaveTangent=5.8572, tangentMode=RCTM_User, time=1, value=2}
        ]}}})
    Index 1 Param 3=&FloatFromCurve(FloatCurve=DataInterface() {propertyValues {Curve {keys [
         {interpMode=RCIM_Cubic, leaveTangent=0.8368, tangentMode=RCTM_Break, time=0, value=0.3}
         {arriveTangent=5.6433, interpMode=RCIM_Cubic, tangentMode=RCTM_Break, time=1, value=3}
        ]}}})
    Index 1 Param 4=&FloatFromCurve(
     FloatCurve=DataInterface() {propertyValues {Curve {keys [
         {interpMode=RCIM_Cubic, leaveTangent=7.8774, tangentMode=RCTM_Break, time=0, value=0}
         {arriveTangent=-3.9574, interpMode=RCIM_Cubic, leaveTangent=-3.9574, tangentMode=RCTM_User, time=0.17, value=1}
         <at 0.7 0>
        ]}}}
     Scale Curve=0.4
    )
    Module.Index0Param0Write=true
    Module.Index0Param1Write=true
    Module.Index0Param2Write=true
    Module.Index0Param3Write=true
    Module.Index1Param0Write=true
    Module.Index1Param1Write=true
    Module.Index1Param2Write=true
    Module.Index1Param3Write=true
    Write Parameter Index 1=true
   )
   &ScaleColor(
    Scale Alpha=&FloatFromCurve(
     FloatCurve=DataInterface() {propertyValues {Curve {keys [
         {arriveTangent=14.2646, interpMode=RCIM_Cubic, leaveTangent=14.2646, tangentMode=RCTM_User, time=0, value=0}
         {arriveTangent=-4.1895, interpMode=RCIM_Cubic, leaveTangent=-4.1895, tangentMode=RCTM_User, time=0.25, value=1.2}
         {arriveTangent=-3.3333, interpMode=RCIM_Cubic, leaveTangent=-3.3333, tangentMode=RCTM_User, time=1, value=0}
        ]}}}
     Scale Curve=0.75
    )
    ScaleA=true
    ScaleRGB=true
   )
  ]
  rendererClasses [&NiagaraMeshRendererProperties]
  renderers [{rendererClass=&NiagaraMeshRendererProperties, rendererIndex=0}]
 }
 Sparks {
  emitterUpdateScript [
   &EmitterState(
    Enable Distance Culling=true
    MaxDistance=15000
    Module.bUseMaxDistance=true
    Module.bUseMinDistance=false
    Scalability Mode=::NewEnumerator1
   )
   &SpawnBurst_Instantaneous(
    Spawn Count=&RandomRangeInt(Minimum=125, Maximum=150)
    Spawn Probability=&MakeCustomFloatFromBool(Bool=$User.IncludeSparks, True Float=1, False Float=0)
    Use Spawn Probability=true
    UseLoopCountLimit=false
   )
  ]
  particleSpawnScript [
   &InitializeParticle(
    Color=$User.FXColor
    Lifetime Max=1
    Lifetime Min=0.15
    Lifetime Mode=::NewEnumerator1
    Sprite Size Max=<6 6>
    Sprite Size Min=<2 2>
    Sprite Size Mode=::NewEnumerator2
    UsePositionOffset=false
    Write Material Random=true
   )
   &ShapeLocation(
    MaskSpawnGroup=false
    Override Seed=false
    Rotation Mode=::NewEnumerator1
    Sphere Radius=&Multiply_Float(A=0.5, B=$User.FXScale)
    Sphere Surface Distribution=0.75
    UseOwnerScale=true
   )
   &AddVelocity(
    Module.Use Falloff Distance=false
    Override Seed=false
    Velocity Mode=::NewEnumerator1
    Velocity Speed=&Lerp_Float(
     A=500
     B=3000
     Alpha=&FloatFromCurve(
      FloatCurve=DataInterface() {propertyValues {Curve {keys [
          {interpMode=RCIM_Cubic, tangentMode=RCTM_User, time=0, value=0}
          {arriveTangent=3.1561, interpMode=RCIM_Cubic, leaveTangent=3.1561, tangentMode=RCTM_User, time=1, value=1}
         ]}}}
      CurveIndex=&RandomRangeFloat(Minimum=0, Maximum=1)
     )
    )
   )
   &AddVelocity(Velocity=&Multiply_Vector(
     A=&Lerp_Vector(
      A=<0 0 200>
      B=<0 0 1500>
      Alpha=&FloatFromCurve(
       FloatCurve=DataInterface() {propertyValues {Curve {keys [
           {interpMode=RCIM_Cubic, tangentMode=RCTM_User, time=0, value=0}
           {arriveTangent=3.1561, interpMode=RCIM_Cubic, leaveTangent=3.1561, tangentMode=RCTM_User, time=1, value=1}
          ]}}}
       CurveIndex=&RandomRangeFloat(Minimum=0, Maximum=1)
      )
     )
     B=<1 1 1>
    )) {module=AddVelocity001}
  ]
  particleUpdateScript [
   &ParticleState()
   &AccelerationForce(Acceleration=&RandomRangeVector(Minimum=<0 0 -800>, Maximum=<0 0 -1100>))
   &CurlNoiseForce(
    Noise Frequency=25
    Noise Strength=&FloatFromCurve(
     FloatCurve=DataInterface() {propertyValues {Curve {keys [{interpMode=RCIM_Cubic, tangentMode=RCTM_Break, time=0, value=1}, {interpMode=RCIM_Cubic, tangentMode=RCTM_Break, time=1, value=0}]}}}
     Scale Curve=1000
    )
    Pan Noise=false
   )
   &Drag(
    Drag=&FloatFromCurve(
     FloatCurve=DataInterface() {propertyValues {Curve {keys [
         {interpMode=RCIM_Cubic, leaveTangentWeight=0.5018, tangentMode=RCTM_Break, time=0, value=1}
         {interpMode=RCIM_Cubic, tangentMode=RCTM_Break, time=0.2, value=0}
        ]}}}
     Scale Curve=15
    )
    Use Linear Drag=true
    Use Rotational Drag=false
   )
   &SolveForcesAndVelocity(Clamp Velocity=false, Limit Acceleration=false)
   &ScaleSpriteSize(
    Uniform Curve Scale=2
    Uniform Curve Sprite Scale=DataInterface() {propertyValues {Curve {keys [{interpMode=RCIM_Cubic, tangentMode=RCTM_Break, time=0, value=1}, {interpMode=RCIM_Cubic, tangentMode=RCTM_Break, time=1, value=0.2}]}}}
   )
   &ScaleSpriteSizeBySpeed(Max Scale Factor=<1 5>)
   &ScaleColor(
    Scale Alpha=&FloatFromCurve(FloatCurve=DataInterface() {propertyValues {Curve {keys [
         {interpMode=RCIM_Cubic, tangentMode=RCTM_User, time=0.0048, value=0.438}
         {arriveTangent=0.3021, interpMode=RCIM_Cubic, leaveTangent=0.3021, tangentMode=RCTM_User, time=0.0983, value=1.5996}
         {arriveTangent=-1.0514, interpMode=RCIM_Cubic, leaveTangent=-1.0514, tangentMode=RCTM_User, time=0.3368, value=0.5383}
         {arriveTangent=-0.8117, interpMode=RCIM_Cubic, leaveTangent=-0.8117, tangentMode=RCTM_User, time=0.5811, value=1.092}
         <at 1 0>
        ]}}})
    Scale RGB=&VectorFromCurve(VectorCurve=DataInterface() {propertyValues {
       XCurve {keys [<at 0 16 RCIM_Cubic>,<at 0.8 1 RCIM_Cubic>]}
       YCurve {keys [<at 0 16 RCIM_Cubic>,<at 0.8 0.1249 RCIM_Cubic>]}
       ZCurve {keys [<at 0 16 RCIM_Cubic>,<at 0.8 0.2547 RCIM_Cubic>]}
      }})
    ScaleA=true
    ScaleRGB=true
   )
   &ScaleColor(
    Scale RGB=&VectorFromFloat(Value=&Lerp_Float(
      A=1
      B=30
      Alpha=&FloatFromCurve(
       FloatCurve=DataInterface() {propertyValues {Curve {keys [
           {interpMode=RCIM_Cubic, tangentMode=RCTM_Break, time=0.4, value=0}
           {arriveTangentWeight=0.8501, interpMode=RCIM_Cubic, tangentMode=RCTM_Break, time=1, value=1}
          ]}}}
       CurveIndex=&RandomRangeFloat(Minimum=0, Maximum=1)
      )
     ))
    ScaleA=false
    ScaleRGB=true
   ) {module=ScaleColor001}
   &ScaleSpriteSize(
    Scale Sprite Size Mode=::NewEnumerator0
    Uniform Scale Factor=/Game/Effects/Niagara/DynamicInputs/Scalability/FN_ScalabilityFloatModifier(
     Value at DistanceFraction 0:NiagaraFloat=1
     Value at DistanceFraction 1:NiagaraFloat=3
     DistanceRamp:NiagaraDataInterfaceCurve=DataInterface() {propertyValues {Curve {keys [<at 0.1 0>,<at 0.5 1>]}}}
    )
   ) {module=ScaleSpriteSize001}
  ]
  rendererClasses [&NiagaraSpriteRendererProperties]
  renderers [{rendererClass=&NiagaraSpriteRendererProperties}]
 }
 SphereElemental {
  emitterUpdateScript [
   &EmitterState()
   &SpawnBurst_Instantaneous(
    Spawn Count=1
    Spawn Probability=&MakeCustomFloatFromBool(
     Bool=&SetBoolByIntComparison(A=$User.FXElement, B=0, Comparison Type=::NewEnumerator4)
     True Float=0
     False Float=1
    )
    Use Spawn Probability=true
   )
   /SetVariables(
    Emitter.PanSpeed:NiagaraFloat=&SelectFloatFromArray(Direct Array Index=$User.FXElement, Float Selection Array=DataInterface() {propertyValues {FloatData [1,1,0,0,2,1]}})
    Emitter.Tile_U:NiagaraFloat=&SelectFloatFromArray(Direct Array Index=$User.FXElement, Float Selection Array=DataInterface() {propertyValues {FloatData [10,10,2,2,5,200]}})
    Emitter.UV_Distortion:NiagaraFloat=&SelectFloatFromArray(Direct Array Index=$User.FXElement, Float Selection Array=DataInterface() {propertyValues {FloatData [4,4,10,10,2,1]}})
   ) {isSetParameters=true}
  ]
  particleSpawnScript [
   &InitializeParticle(
    Color=$User.FXColor
    Lifetime=&Multiply_Float(
     A=&SelectFloatFromArray(Direct Array Index=$User.FXElement, Float Selection Array=DataInterface() {propertyValues {FloatData [0.8,0.8,0.8,0.8,1.25,1]}})
     B=$User.FXTime
    )
    Mesh Scale Mode=::NewEnumerator3
    Mesh Uniform Scale=&Multiply_Float(A=0.028, B=$User.FXScale)
    UsePositionOffset=false
    Write Material Random=true
   )
   &InitialMeshOrientation(Mesh Axis To Orient=<0 0 1>, Mesh Orientation Mode=::NewEnumerator2, Use Rotation Vector=false)
  ]
  particleUpdateScript [
   &ParticleState()
   &ScaleMeshSize(Scale Factor=&VectorFromFloat(Value=&Lerp_Float(
      A=0.5
      B=1
      Alpha=&FloatFromCurve(FloatCurve=DataInterface() {propertyValues {Curve {keys [
           {arriveTangent=3.5012, interpMode=RCIM_Cubic, leaveTangent=3.5012, tangentMode=RCTM_User, time=0, value=0}
           {interpMode=RCIM_Cubic, tangentMode=RCTM_User, time=1, value=1}
          ]}}})
     )))
   &ScaleColor(
    Scale Alpha=&FloatFromCurve(FloatCurve=DataInterface() {propertyValues {Curve {keys [{interpMode=RCIM_Cubic, tangentMode=RCTM_User, time=0, value=1}, {interpMode=RCIM_Cubic, tangentMode=RCTM_User, time=1, value=0}]}}})
    Scale RGB=&VectorFromFloat(Value=&Lerp_Float(
      A=0
      B=30
      Alpha=&Multiply_Float(
       A=1
       B=&FloatFromCurve(FloatCurve=DataInterface() {propertyValues {Curve {keys [
            {arriveTangent=-2.6334, interpMode=RCIM_Cubic, leaveTangent=-2.6334, tangentMode=RCTM_User, time=0, value=1}
            {interpMode=RCIM_Cubic, tangentMode=RCTM_User, time=1, value=0}
           ]}}})
      )
     ))
    ScaleA=true
    ScaleRGB=true
   )
   &DynamicMaterialParameters(
    Index 0 Param 1=&FloatFromCurve(
     FloatCurve=DataInterface() {propertyValues {Curve {keys [
         {arriveTangent=0.3148, interpMode=RCIM_Cubic, leaveTangent=0.3148, tangentMode=RCTM_User, time=0.5, value=0.1}
         {arriveTangent=4.0372, interpMode=RCIM_Cubic, leaveTangent=4.0372, tangentMode=RCTM_User, time=1, value=1}
        ]}}}
     Scale Curve=&SelectFloatFromArray(Direct Array Index=$User.FXElement, Float Selection Array=DataInterface() {propertyValues {FloatData [1,1,4,3,2,1]}})
    )
    Module.Index0Param0Write=true
    Module.Index0Param1Write=true
    Module.Index0Param2Write=true
    Module.Index0Param3Write=true
   )
  ]
  rendererClasses [&NiagaraMeshRendererProperties]
  renderers [{rendererClass=&NiagaraMeshRendererProperties, rendererIndex=0}]
 }
 SphereGlow {
  emitterUpdateScript [
   &EmitterState()
   &SpawnBurst_Instantaneous(Spawn Time=0.1, Use Spawn Probability=false, UseLoopCountLimit=false)
  ]
  particleSpawnScript [
   &InitializeParticle(
    Color=$User.FXColor
    Lifetime=$User.FXTime
    Mesh Scale Mode=::NewEnumerator3
    Mesh Uniform Scale=&Multiply_Float(A=0.0065, B=$User.FXScale)
    UsePositionOffset=false
    Write Material Random=true
   )
   &InitialMeshOrientation(
    Mesh Orientation Mode=::NewEnumerator7
    Rotation=&RandomRangeVector(Minimum=<0 0 0>, Maximum=<1 0 0>)
    Use Rotation Vector=true
   )
  ]
  particleUpdateScript [
   &ParticleState()
   &Timeline(
    Play Rate=&FloatFromCurve(
     FloatCurve=DataInterface() {propertyValues {Curve {keys [{interpMode=RCIM_Cubic, tangentMode=RCTM_Break, time=0, value=1}, {tangentMode=RCTM_User, time=1, value=0}]}}}
     Scale Curve=2
    )
    Rewind Mode:ENiagara_TimelineMode=::NewEnumerator1
    Timeline End=1000
   )
   &ScaleMeshSize(Scale Factor=&VectorFromFloat(Value=&FloatFromCurve(FloatCurve=DataInterface() {propertyValues {Curve {keys [
          {arriveTangent=6.7706, interpMode=RCIM_Cubic, leaveTangent=6.7706, tangentMode=RCTM_User, time=0, value=0}
          {arriveTangent=1.0094, interpMode=RCIM_Cubic, leaveTangent=1.0094, tangentMode=RCTM_User, time=0.2, value=0.7}
          {arriveTangent=0.0224, interpMode=RCIM_Cubic, leaveTangent=0.0224, tangentMode=RCTM_User, time=1, value=1}
         ]}}})))
   &ScaleColor(
    Scale Alpha=&FloatFromCurve(FloatCurve=DataInterface() {propertyValues {Curve {keys [
         {arriveTangent=9.0725, interpMode=RCIM_Cubic, leaveTangent=9.0725, tangentMode=RCTM_User, time=0, value=0}
         {arriveTangent=-7.5496, interpMode=RCIM_Cubic, leaveTangent=-7.5496, tangentMode=RCTM_User, time=0.15, value=0.9847}
         {arriveTangent=-0.0405, interpMode=RCIM_Cubic, leaveTangent=-0.0405, tangentMode=RCTM_User, time=0.5, value=0}
        ]}}})
    Scale RGB=&VectorFromFloat(Value=2)
    ScaleA=true
    ScaleRGB=true
   )
   &CameraOffset(Camera Offset Amount=10)
   &DynamicMaterialParameters(
    Index 0 Param 1=10
    Index 0 Param 3=15
    Index 0 Param 4=&Multiply_Float(A=$Particles.Timeline.Time, B=1)
    Module.Index0Param0Write=true
    Module.Index0Param1Write=true
    Module.Index0Param2Write=true
    Module.Index0Param3Write=true
   )
  ]
  rendererClasses [&NiagaraMeshRendererProperties]
  renderers [{rendererClass=&NiagaraMeshRendererProperties, rendererIndex=0}]
 }
 SphereMain {
  emitterUpdateScript [
   &EmitterState()
   &SpawnBurst_Instantaneous(Use Spawn Probability=false, UseLoopCountLimit=false)
   /SetVariables(
    Emitter.PlasmaAmount:NiagaraFloat=&SelectFloatFromArray(Direct Array Index=$User.FXElement, Float Selection Array=DataInterface() {propertyValues {FloatData [0.1,0.4,0.2,0.2,0.4,0.2]}})
    Emitter.UVScale:Vector4f=/Niagara/DynamicInputs/Arrays/SelectVector4_FromArray(
     Array Sampling Mode:ENiagaraArraySamplingMode=::NewEnumerator1
     Direct Array Index:NiagaraInt32=$User.FXElement
     Interpolation:NiagaraFloat=?
     Direct Array Mode:ENiagara_WrapClamp=::NewEnumerator0
     Evaluation Type:ENiagaraRandomnessEvaluation=::NewEnumerator0
     Normalized Interpolation:NiagaraBool=true
     Recalculate Random Each Loop:NiagaraBool=false
     Randomness Mode:ENiagaraRandomnessMode=::NewEnumerator0
     Random Seed:NiagaraInt32=?
     Fixed Random Seed:NiagaraBool=false
     Vector4 Selection Array:NiagaraDataInterfaceArrayFloat4=DataInterface() {propertyValues {GpuSyncMode=SyncCpuToGpu, MaxElements=0, FloatData [<0.1 0.1 0 0>, <0.2 0.5 0 0>, <0.1 0.1 0 0>]}}
    )
   ) {isSetParameters=true}
  ]
  particleSpawnScript [
   &InitializeParticle(
    Color=$User.FXColor
    Lifetime=$User.FXTime
    Mesh Scale Mode=::NewEnumerator3
    Mesh Uniform Scale=&Multiply_Float(A=0.032, B=$User.FXScale)
    UsePositionOffset=false
    Write Material Random=true
   )
   &InitialMeshOrientation(Mesh Orientation Mode=::NewEnumerator2, Use Rotation Vector=false)
  ]
  particleUpdateScript [
   &ParticleState()
   &ScaleMeshSize(Scale Factor=&VectorFromFloat(Value=&Lerp_Float(
      A=0.5
      B=1
      Alpha=&FloatFromCurve(FloatCurve=DataInterface() {propertyValues {Curve {keys [
           {interpMode=RCIM_Cubic, leaveTangentWeight=0.9987, tangentMode=RCTM_User, time=0, value=0}
           {interpMode=RCIM_Cubic, tangentMode=RCTM_User, time=1, value=1}
          ]}}})
     )))
   &DynamicMaterialParameters(
    Index 0 Param 1=&FloatFromCurve(FloatCurve=DataInterface() {propertyValues {Curve {keys [<at 0 0 RCIM_Cubic>, {arriveTangent=3.474, interpMode=RCIM_Cubic, leaveTangent=3.474, tangentMode=RCTM_User, time=1, value=1}]}}})
    Module.Index0Param0Write=true
    Module.Index0Param1Write=false
    Module.Index0Param2Write=true
    Module.Index0Param3Write=true
   )
   &ScaleColor(
    Scale Alpha=&FloatFromCurve(
     FloatCurve=DataInterface() {propertyValues {Curve {keys [
         {interpMode=RCIM_Cubic, tangentMode=RCTM_User, time=0, value=1}
         {arriveTangent=-2.281, interpMode=RCIM_Cubic, leaveTangent=-2.281, tangentMode=RCTM_User, time=1, value=0}
        ]}}}
     Scale Curve=$User.FXIntensity
    )
    Scale RGB=&VectorFromFloat(Value=&Multiply_Float(
      A=&Lerp_Float(
       A=5
       B=35
       Alpha=&FloatFromCurve(FloatCurve=DataInterface() {propertyValues {Curve {keys [
            {interpMode=RCIM_Cubic, leaveTangentWeight=0.1521, tangentMode=RCTM_User, time=0, value=1}
            {arriveTangent=-1.146, leaveTangent=-1.146, time=0.4, value=0.2}
            {arriveTangent=-0.0405, interpMode=RCIM_Cubic, leaveTangent=-0.0405, tangentMode=RCTM_User, time=1, value=0}
           ]}}})
      )
      B=$User.FXIntensity
     ))
    ScaleA=true
    ScaleRGB=true
   )
  ]
  rendererClasses [&NiagaraMeshRendererProperties]
  renderers [{rendererClass=&NiagaraMeshRendererProperties, rendererIndex=0}]
 }
}
summary {
 systemName=NS_TEMP_RadialExplode
 emitters [
  {emitterName=BrightStreaks, rendererClasses [&NiagaraSpriteRendererProperties]}
  {emitterName=SphereMain, rendererClasses [&NiagaraMeshRendererProperties]}
  {emitterName=SphereElemental, rendererClasses [&NiagaraMeshRendererProperties]}
  {emitterName=LightAndGlow, rendererClasses [&NiagaraSpriteRendererProperties, &NiagaraLightRendererProperties]}
  {bEnabled=false, emitterName=PostMat, rendererClasses [&NiagaraMeshRendererProperties]}
  {emitterName=SphereGlow, rendererClasses [&NiagaraMeshRendererProperties]}
  {emitterName=DEBUG, rendererClasses [&NiagaraSpriteRendererProperties]}
  {emitterName=Sparks, rendererClasses [&NiagaraSpriteRendererProperties]}
 ]
 userVariables [
  User.FXColor {type=LinearColor, description="", defaultValue {struct=LinearColor, value=<rgba 0.9087 0.0025 1 1>}}
  User.FXElement {type=NiagaraInt32}
  User.FXIntensity {type=NiagaraFloat, defaultValue=1, description=""}
  User.FXScale {type=NiagaraFloat, defaultValue=500, description=""}
  User.FXTime {type=NiagaraFloat, defaultValue=0.5, description=""}
  User.IncludeSparks {type=NiagaraBool, defaultValue=true}
  User.ShowDebugRing {type=NiagaraBool}
 ]
}
